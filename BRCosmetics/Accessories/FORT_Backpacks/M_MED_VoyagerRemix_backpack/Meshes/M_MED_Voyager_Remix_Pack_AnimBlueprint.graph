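# Animation Blueprint for the Voyager Remix backpack cosmetic, adding secondary motion to its dangling game pad.
# The ref pose gets an additive pitch tilt on the pack bone in bone space.
# An Anim Dynamics node then simulates the dyn_game_pad bone as one heavily damped, angle-limited body with overridden gravity and scaled-down component motion.
# The event graph holds only unconnected default nodes, so there is no scripted logic.
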AnimBlueprint /BRCosmetics/Accessories/FORT_Backpacks/M_MED_VoyagerRemix_backpack/Meshes/M_MED_Voyager_Remix_Pack_AnimBlueprint
target_skeleton=&/M_MED_Voyager_Remix_Pack_Skeleton
graphs [
 AnimGraph {nodes [
   n1 = "Transform (Modify) Bone - Bone: pack"(ComponentPose=Local To Component(LocalPose=Local Space Ref Pose()), Translation="0,0,0", Rotation="-10,0.000,0.000") {
    rotationMode=BMM_Additive
    rotationSpace=BCS_BoneSpace
    alphaBoolBlend {blendOption=Linear}
    boneToModify {boneName=pack}
    rotation {pitch=-10}
   }
   n2 = Anim Dynamics - Bone: dyn_game_pad(ComponentPose=@n1, Alpha=1, ExternalForce="0,0,0") {
    angularDampingOverride=0.99
    bGravityOverrideInSimSpace=true
    bOverrideAngularDamping=true
    bUseGravityOverride=true
    alphaBoolBlend {blendOption=Linear}
    boundBone {boneName=dyn_game_pad}
    componentAppliedLinearAccClamp=<1000 1000 1000>
    componentLinearAccScale=<0.2 0.2 0.4>
    componentLinearVelScale=<0.2 0.2 0.4>
    gravityOverride {z=-980}
    physicsBodyDefinitions [{
      sphereCollisionRadius=0
      boundBone {boneName=dyn_game_pad}
      boxExtents=<5 7 7>
      constraintSetup {angularLimitsMax=<0 6 6>, angularLimitsMin=<0 -6 -6>, angularTarget=<0 0 0>}
      localJointOffset=<2 0 0>
     }]
   }
   Output Pose(Result=Component To Local(ComponentPose=@n2)) {layerGroup=None}
  ]}
 EventGraph {nodes [
   TryGetPawnOwner()
   Event BlueprintUpdateAnimation()
  ]}
]
variables []
